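# Unlit, two-sided translucent material that draws the bounds of a placed Creative barrier device.
# Rotated world-aligned stripes, staggered dots that fade with camera distance, and Fresnel form a mask blending fill and stripe/edge colors.
# A placer texture, sampled with object-scale-corrected UVs from channels 1 and 2, supplies the bottom gradient.
# Opacity controls dim backfaces, with near-camera and depth fades applied only at high quality.
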Material /Game/Creative/Devices/DeviceZone_Base/Materials/M_CP_Barrier_Bounds_Placed
graph {
 properties {
  bEnableResponsiveAA=true
  blendMode=BLEND_Translucent
  pixelDepthOffsetMode=PDOM_Legacy
  shadingModel=MSM_Unlit
  twoSided=true
  usedShadingModels=Unlit
 }
 nodes [
  n1 = Constant() {r=1}
  n2 = ComponentMask(None=TextureCoordinate() {coordinateIndex=2}) {r=true}
  n3 = TextureCoordinate() {coordinateIndex=2}
  n4 = AppendVector(
   A=Divide(
    A=Constant() {r=1}
    B=MaterialFunctionCall() {materialFunction=&ObjectScale}.Scale Y
   )
   B=Divide(
    A=Constant() {desc="0.85 is default, not 1", r=1}
    B=MaterialFunctionCall() {materialFunction=&ObjectScale}.Scale Z
   )
  )
  n5 = Constant() {r=1}
  n6 = Subtract(A=Add(A=Divide(A=@n3, B=@n4), B=@n5), B=Divide(A=@n5, B=@n4) {desc=Scale the center offset by the scaling factor}) {desc=Re offset to maintain center}
  n7 = Constant()
  n8 = Constant() {r=1}
  n9 = ComponentMask(None=TextureCoordinate() {coordinateIndex=1}) {r=true}
  n10 = TextureCoordinate() {coordinateIndex=1}
  n11 = AppendVector(
   A=Divide(
    A=Constant() {r=1}
    B=MaterialFunctionCall() {materialFunction=&ObjectScale}.Scale X
   )
   B=Divide(
    A=Constant() {desc="0.85 is default, not 1", r=1}
    B=MaterialFunctionCall() {materialFunction=&ObjectScale}.Scale Z
   )
  )
  n12 = Constant() {r=1}
  n13 = Subtract(A=Add(A=Divide(A=@n10, B=@n11), B=@n12), B=Divide(A=@n12, B=@n11) {desc=Scale the center offset by the scaling factor}) {desc=Re offset to maintain center}
  n14 = Constant()
  n15 = TextureSample(UVs=Add(A=If(A=@n1, B=@n2, "A > B"=@n6, "A < B"=@n7) {desc=0}, B=If(A=@n8, B=@n9, "A > B"=@n13, "A < B"=@n14) {desc=0})) {
   samplerSource=SSM_Clamp_WorldGroupSettings
   texture=/Game/Creative/Devices/DeviceZone_Base/Materials/Textures/T_CP_DeviceZone_Bounds_01_BarrierPlacer_D
  }
  n16 = Multiply(A=@n15.G, B=ScalarParameter() {defaultValue=0.5, group=Visibility, parameterName=BottomGradientOpacity})
  n17 = VectorParameter() {group=Colors, parameterName=StripeColor, sortPriority=1, defaultValue=<rgba 0.6038 0.1559 0 1>}
  n18 = VectorParameter() {group=Colors, parameterName=EdgeGradientColor, sortPriority=2, defaultValue=<rgba 0.2961 0.0762 0.0103 1>}
  n19 = TextureCoordinate()
  n20 = Constant() {r=0.5}
  n21 = Constant() {r=4}
  n22 = StaticBool()
  n23 = NamedRerouteDeclaration(None=OneMinus(None=MaterialFunctionCall(UVs=@n19, Radius=@n20, Density=@n21, Invert Density=@n22) {materialFunction=&RadialGradientExponential}.RadialGradientExponential)) {name=EdgeGradientMask, nodeColor=<rgba 1 0 0.0703 1>}
  n24 = MaterialFunctionCall!(WorldScale=VectorParameter() {parameterName=WorldScale, defaultValue=<rgba 20 20 20 0>}.RGB, TransformToLocalSpace) {materialFunction=/Game/Items/Traps/Materials/MF_Creative_WorldAlignedUVs}
  n25 = Rotator(Coordinate=@n24.Result, Time=ScalarParameter() {defaultValue=9.3, group=Stripes, parameterName=Time})
  n26 = SphereMask(
   A=MaterialFunctionCall(Float2=Frac(None=@n25)) {materialFunction=&BreakOutFloat2Components}.R
   B=Constant() {r=0.5}
   Radius=ScalarParameter() {defaultValue=0.2, group=Stripes, parameterName=StripeRadius}
   Hardness=ScalarParameter() {defaultValue=0.9, group=Stripes, parameterName=StripeHarndess}
  )
  n27 = Divide(A=@n24.Result, B=ScalarParameter() {defaultValue=0.2, group=Dots, parameterName=DotsRepeat})
  n28 = WorldPosition()
  n29 = ScalarParameter() {defaultValue=19.48, group=Dots, parameterName=WorldScaleDots}
  n30 = Divide(
   A=ComponentMask(None=Divide(A=@n28.XYZ, B=@n29)) {b=true}
   B=Multiply(A=ScalarParameter() {defaultValue=0.2, group=Dots, parameterName=DotsRepeat}) {constB=2}
  ) {constB=16}
  n31 = LinearInterpolate(A=Add(A=Constant2Vector() {r=0.5}, B=@n27), B=@n27, Alpha=Round(None=Frac(None=@n30)))
  n32 = ScalarParameter() {defaultValue=0.1, group=Dots, parameterName=Radius}
  n33 = ScalarParameter() {defaultValue=0.25, group=Dots, parameterName=RadiusLarge}
  n34 = TextureCoordinate()
  n35 = SphereMask(
   A=@n34
   B=Constant() {desc=0.5, r=0.5}
   Radius=ScalarParameter() {defaultValue=0.75, desc=0.5, parameterName=BGRadius}
   Hardness=Constant() {desc=0}
  )
  n36 = CameraPositionWS()
  n37 = WorldPosition()
  n38 = Divide(
   A=Add(A=Distance(A=@n36, B=@n37.XYZ), B=ScalarParameter() {defaultValue=-100, group=Visibility, parameterName=DistanceBias})
   B=ScalarParameter() {defaultValue=800, group=Visibility, parameterName=FadeScale}
  )
  n39 = Multiply(
   A=SphereMask(
    A=Frac(None=@n31)
    B=Constant() {r=0.5}
    Radius=LinearInterpolate(A=@n32, B=@n33, Alpha=@n35)
    Hardness=ScalarParameter() {defaultValue=0.5, group=Dots, parameterName=DotsHardness}
   )
   B=OneMinus(None=Saturate(None=@n38))
  )
  n40 = ScalarParameter() {defaultValue=0.5, group=Visibility, parameterName=Fresnel Brightness}
  n41 = ScalarParameter() {defaultValue=4, group=Visibility, parameterName=Fresnel Exponent}
  n42 = Multiply(
   A=Add(
    A=Add(A=@n26, B=@n39)
    B=MaterialFunctionCall(Brightness=@n40, Exponent=@n41) {materialFunction=&MF_Fresnel_FX}.Fresnel Inverted
   )
   B=ScalarParameter() {defaultValue=1, group=Colors, parameterName=StripeOpacity, sortPriority=10}
  )
  n43 = Multiply(
   A=OneMinus(None=NamedRerouteDeclaration(None=@n16) {name=BottomGradientMask, nodeColor=<rgba 1 0 0.8203 1>})
   B=LinearInterpolate(
    A=VectorParameter() {group=Colors, parameterName=FillColor, sortPriority=3, defaultValue=<rgba 0.0212 0.008 0 1>}.RGB
    B=LinearInterpolate(A=@n17.RGB, B=@n18.RGB, Alpha=@n23)
    Alpha=NamedRerouteDeclaration(None=@n42) {name=StripesDotsMask, nodeColor=<rgba 0 1 0.2031 1>}
   )
  )
  n44 = VectorParameter() {group=Colors, parameterName=BottomGradientColor, sortPriority=4, defaultValue=<rgba 0.491 0.3916 0.1119 1>}
  LinearInterpolate(
   A=@n43
   B=Add(A=@n44.RGB, B=@n43)
   Alpha=Multiply(A=NamedRerouteUsage(), B=ScalarParameter() {defaultValue=1, group=Colors, parameterName=BottomGradientIntensity, sortPriority=13})
  )
  MaterialFunctionCall(Float2=@n34) {materialFunction=&BreakOutFloat2Components}
  n45 = Multiply(A=NamedRerouteUsage(), B=ScalarParameter() {defaultValue=1, group=Colors, parameterName=BottomGradientIntensity, sortPriority=13})
  n46 = Multiply(A=NamedRerouteUsage(), B=ScalarParameter() {defaultValue=1, group=Colors, parameterName=EdgeGradientOpacity, sortPriority=11})
  n47 = LinearInterpolate(
   A=LinearInterpolate(
    A=ScalarParameter() {defaultValue=0.2, group=Colors, parameterName=FillOpacity, sortPriority=12}
    Alpha=LinearInterpolate(A=@n45, Alpha=@n46)
   )
   Alpha=NamedRerouteUsage()
  )
  n48 = LinearInterpolate(
   A=Multiply(A=ScalarParameter() {defaultValue=0.2, group=Visibility, parameterName=BacksideOpacity}, B=@n47)
   B=@n47
   Alpha=Saturate(None=TwoSidedSign())
  )
  n49 = Multiply(
   A=@n48
   B=MaterialFunctionCall(Distance=ScalarParameter() {defaultValue=200, group=Visibility, parameterName="Fade_Start_Distance (S)"}) {materialFunction=&Near_Camera_Fade}.Result
  )
  n50 = ScalarParameter() {defaultValue=15, group=Visibility, parameterName=FadeDistance}
  Multiply(
   A=QualitySwitch(Default=@n48, Low=@n48, High=DepthFade(Opacity=@n49, FadeDistance=@n50), Medium=@n48)
   B=ScalarParameter() {defaultValue=1, group=Colors, parameterName=Color Boost}
  )
 ]
}
