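# Translucent, two-sided unlit material for the bounds of a placed Creative damage volume, drawing an animated fiery energy surface.
# Pulsing noise, partly projected from world position by rounded normals, combines with Fresnel, depth-behind glow, a height gradient and panning fire textures, simplified on mobile.
# That mask blends two colors, tinted by particle color, with near-camera fade only at High quality.
# UV gradients and particle alpha set opacity before MF_CreativeMoveTool processes the attributes.
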Material /Game/Creative/Devices/DeviceZone_Base/Materials/M_CP_DamageVolume_Bounds_Placed
graph {
 properties {
  allowTranslucentCustomDepthWrites=true
  bEnableResponsiveAA=true
  bUseMaterialAttributes=true
  bUsedWithMeshParticles=true
  blendMode=BLEND_Translucent
  pixelDepthOffsetMode=PDOM_Legacy
  shadingModel=MSM_Unlit
  twoSided=true
  usedShadingModels=Unlit
 }
 nodes [
  n1 = Add(
   A=Multiply(A=Time(), B=ScalarParameter() {defaultValue=-0.3, parameterName=NoiseSineSpeed})
   B=MaterialFunctionCall(Float2=TextureCoordinate() {vTiling=0.7}) {materialFunction=&BreakOutFloat2Components}.G
  )
  Cosine(None=@n1) {period=2}
  LinearInterpolate()
  n2 = ParticleColor()
  n3 = VectorParameter() {parameterName=BGColor, defaultValue=<rgba 0.8906 0 0.0148 0>}
  n4 = Multiply(A=ComponentMask(None=TextureCoordinate() {vTiling=2}) {g=true}, B=Constant() {r=0.1}) {constB=0.25}
  n5 = VectorParameter() {parameterName=Circles, defaultValue=<rgba 10 0 0.2122 0>}
  n6 = TextureCoordinate() {coordinateIndex=1, unMirrorU=true}
  n7 = Constant() {r=1}
  n8 = AppendVector(A=ScalarParameter() {parameterName=PanU}, B=ScalarParameter() {parameterName=PanV})
  n9 = TextureSampleParameter2D(UVs=Divide(A=Panner(Coordinate=@n6, Time=@n7, Speed=@n8) {speedX=5}, B=ScalarParameter() {defaultValue=950, parameterName=RepeatNoiseA})) {parameterName=Noise, texture=/Game/Global/Textures/Default/Noise/T_Dudebro_Noise_01}
  n10 = MaterialFunctionCall(Float3=Multiply(A=WorldPosition().XYZ) {constB=-1}) {materialFunction=&BreakOutFloat3Components}
  n11 = VertexNormalWS()
  n12 = MaterialFunctionCall(Float3=Abs(None=Round(None=@n11))) {materialFunction=&BreakOutFloat3Components}
  n13 = LinearInterpolate(
   A=LinearInterpolate(A=AppendVector(A=@n10.G, B=@n10.G), B=AppendVector(A=@n10.R, B=@n10.R), Alpha=@n12.G)
   B=AppendVector(A=@n10.R, B=@n10.R)
   Alpha=@n12.B
  )
  n14 = TextureSampleParameter2D(UVs=Divide(A=Panner(Coordinate=@n13, Speed=@n8) {speedX=5}, B=ScalarParameter() {defaultValue=-1500, parameterName=RepeatNoiseB})) {parameterName=Noise, texture=&T_Dudebro_Noise_01}
  n15 = Sine(None=@n1) {period=2}
  n16 = Multiply(
   A=SphereMask(
    A=Multiply(A=@n9.R, B=@n14.R)
    B=ConstantBiasScale(None=@n15)
    Radius=ScalarParameter() {defaultValue=0.3, parameterName=NoiseRadius}
    Hardness=ScalarParameter() {defaultValue=0.6, parameterName=NoiseHardness}
   )
   B=ScalarParameter() {defaultValue=1, parameterName=NoiseScale}
  )
  n17 = MaterialFunctionCall(
   Brightness=ScalarParameter() {defaultValue=0.1, parameterName=Fresnel Brightness}
   Exponent=ScalarParameter() {defaultValue=3, parameterName=Fresnel Exponent}
  ) {materialFunction=&MF_Fresnel_FX}
  n18 = MaterialFunctionCall!(Distance Bias, Distance Divisor=ScalarParameter() {defaultValue=1, parameterName="Distance Divisor (S)"}, Scene Depth, Pixel Depth) {materialFunction=/Game/Effects/Fort_Effects/Materials/_Functions/MF_DistanceBehindTranslucency_01}
  n19 = ScalarParameter() {defaultValue=0.1, parameterName=DistanceBehindTranslucencyScale}
  n20 = TextureCoordinate()
  n21 = ScalarParameter() {defaultValue=0.5, parameterName=FlameHeight}
  n22 = ComponentMask(None=Saturate(None=Multiply(A=@n20, B=@n21))) {g=true}
  n23 = Multiply(A=@n22, B=@n22)
  n24 = Multiply(
   A=TextureCoordinate() {uTiling=0.75, vTiling=0.05}
   B=AppendVector(A=ScalarParameter() {defaultValue=7, parameterName=TileU}, B=ScalarParameter() {defaultValue=5, parameterName=TileV})
  )
  n25 = Time()
  n26 = ScalarParameter() {defaultValue=0.1, parameterName=Speed}
  n27 = TextureObject() {texture=/Game/Athena/Environments/Festivus/Materials/T_Fire_01_D}
  n28 = TextureSample(UVs=Panner(Coordinate=@n24, Time=Multiply(A=@n25, B=@n26)) {speedX=-0.75, speedY=2.5}, Tex=@n27)
  n29 = Time()
  n30 = ScalarParameter() {defaultValue=0.2, parameterName=Speed03}
  n31 = TextureSample(UVs=Panner(Coordinate=@n24, Time=Multiply(A=@n29, B=@n30)) {speedX=0.4, speedY=3.25}, Tex=@n27)
  n32 = Multiply(A=Multiply(A=@n13, B=@n24)) {constB=2}
  n33 = Multiply(A=Time(), B=ScalarParameter() {defaultValue=5, parameterName=Speed02})
  n34 = Multiply(A=TextureSample(UVs=Panner(Coordinate=@n32, Time=@n33) {speedX=0.2, speedY=2}, Tex=@n27).R)
  n35 = Multiply(A=Add(A=Multiply(A=@n28.R) {constB=5}, B=Multiply(A=@n31.G) {constB=3}), B=@n34)
  n36 = Multiply(A=@n35, B=DepthFade(Opacity=Constant() {r=10}, FadeDistance=ScalarParameter() {defaultValue=200, parameterName=DepthFadeDist}))
  n37 = Add(
   A=Add(A=Add(A=@n16, B=@n17.Fresnel Inverted), B=Multiply(A=@n18.Normalized Inverted, B=@n19))
   B=Add(A=@n23, B=ShadingPathSwitch(Default=@n36, Mobile=@n34) {desc=Migrated from FeatureLevelSwitch})
  )
  n38 = Multiply(A=@n3.RGB, B=Multiply(A=Saturate(None=@n4), B=LinearInterpolate(A=@n3.RGB, B=@n5.RGB, Alpha=@n37)))
  n39 = Multiply(A=@n3.RGB, B=@n5.RGB)
  n40 = Multiply(A=@n2.RGB, B=Multiply(A=Multiply(A=@n38, B=@n39), B=@n37))
  n41 = Multiply(A=BlackBody!(Temp=Constant() {r=4500}), B=@n40)
  n42 = Multiply(
   A=@n40
   B=MaterialFunctionCall(Distance=ScalarParameter() {defaultValue=200, parameterName="Fade_Start_Distance (S)"}) {materialFunction=&Near_Camera_Fade}.Result
  )
  n43 = Power(Base=ComponentMask(None=TextureCoordinate()) {g=true}, Exp=Constant() {r=5}) {constExponent=10}
  n44 = TextureCoordinate()
  n45 = Power(Base=OneMinus(None=ComponentMask(None=@n44) {g=true}), Exp=ScalarParameter() {defaultValue=5, parameterName=Central Contast}) {constExponent=10}
  n46 = Power(Base=OneMinus(None=LinearInterpolate(A=@n43, B=@n45)), Exp=Constant() {r=15}) {constExponent=10}
  n47 = LinearInterpolate(B=Multiply(A=ScalarParameter() {defaultValue=0.0005, parameterName=Opacity}), Alpha=Power(Base=@n46, Exp=Constant() {r=1}))
  MaterialFunctionCall(MA In=SetMaterialAttributes!(MaterialAttributes, Emissive Color=QualitySwitch(Default=@n41, Low=@n41, High=@n42, Medium=@n41), Opacity=Multiply(A=@n2.A, B=@n47)) {attributeSetTypes [#1, #2]}) {materialFunction=&MF_CreativeMoveTool}
  MaterialFunctionCall() {materialFunction=&LinearSine}
  Multiply(A=@n46, B=@n23)
  Reroute(None=@n35)
  TextureSample(Tex=@n27)
 ]
}
