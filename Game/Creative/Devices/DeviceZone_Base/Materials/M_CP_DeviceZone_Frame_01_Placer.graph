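# Unlit translucent material for a Creative device zone frame, apparently its placer variant, drawn as two tinted regions.
# A switchable box mask thresholds UV gradients against four edge parameters to separate the border from the interior.
# The mask blends Frame color and opacity on the border with Outline color and opacity inside.
# An optional fade, off by default, lowers opacity as camera distance grows.
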Material /Game/Creative/Devices/DeviceZone_Base/Materials/M_CP_DeviceZone_Frame_01_Placer
graph {
 properties {blendMode=BLEND_Translucent, pixelDepthOffsetMode=PDOM_Legacy, shadingModel=MSM_Unlit, usedShadingModels=Unlit}
 nodes [
  n1 = If(
   A=MaterialFunctionCall() {materialFunction=&LinearGradient}.UGradient
   B=ScalarParameter() {defaultValue=0.975, group=Mask, parameterName=Right X}
   "A > B"=Constant() {r=1}
   "A < B"=Constant()
  )
  n2 = If(
   A=MaterialFunctionCall() {materialFunction=&LinearGradient}.UGradient
   B=ScalarParameter() {defaultValue=0.025, group=Mask, parameterName=Left X}
   "A > B"=Constant() {r=1}
   "A < B"=Constant()
  )
  n3 = If(
   A=MaterialFunctionCall() {materialFunction=&LinearGradient}.VGradient
   B=ScalarParameter() {defaultValue=0.975, group=Mask, parameterName=Bottom Y}
   "A > B"=Constant() {r=1}
   "A < B"=Constant()
  )
  n4 = If(
   A=MaterialFunctionCall() {materialFunction=&LinearGradient}.VGradient
   B=ScalarParameter() {defaultValue=0.025, group=Mask, parameterName=Top Y}
   "A > B"=Constant() {r=1}
   "A < B"=Constant()
  )
  n5 = StaticSwitchParameter(True=Multiply(A=Subtract(A=@n1, B=@n2), B=Subtract(A=@n3, B=@n4)), False=Constant3Vector()) {defaultValue=true, parameterName=Enable Box Mask}
  LinearInterpolate(
   A=Multiply(
    A=VectorParameter() {group=Frame, parameterName=Frame Color, defaultValue=<rgba 0.3005 0.5711 1 1>}.RGB
    B=ScalarParameter() {defaultValue=0.5962, group=Frame, parameterName=Frame Color Multiplier}
   )
   B=Multiply(
    A=VectorParameter() {group=Outline, parameterName=Outline Color, defaultValue=<rgba 0.6654 0.7529 0.8308 1>}.RGB
    B=ScalarParameter() {defaultValue=0.85, group=Outline, parameterName=Outline Color Multiplier}
   )
   Alpha=@n5
  )
  n6 = LinearInterpolate(
   A=ScalarParameter() {defaultValue=0.75, group=Frame, parameterName=Frame Color Opacity}
   B=ScalarParameter() {defaultValue=0.75, group=Outline, parameterName=Outline Color Opacity}
   Alpha=@n5
  )
  n7 = Distance(A=CameraPositionWS(), B=WorldPosition().XYZ)
  n8 = ScalarParameter() {defaultValue=-100, group=Distance Fade, parameterName=DistanceBias}
  n9 = Saturate(None=Divide(A=Add(A=@n7, B=@n8), B=ScalarParameter() {defaultValue=1200, group=Distance Fade, parameterName=FadeScale}))
  LinearInterpolate(
   A=@n6
   B=Multiply(A=@n6, B=OneMinus(None=@n9))
   Alpha=ScalarParameter() {group=Distance Fade, parameterName=Enable Distance Fade, sortPriority=1}
  ) {constA=1}
 ]
}
