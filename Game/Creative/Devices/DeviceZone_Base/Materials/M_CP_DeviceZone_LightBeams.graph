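# Unlit, two-sided alpha-composite material for Creative DeviceZone light beams, usable on mesh particles.
# A length-scaled falloff mask shapes double-sided Fresnel, which combines with swept scanlines, two panning noise layers and optional cloud noise.
# The result drives emissive color (gradient or custom primitive data) and opacity, with switches for vertex and particle color.
# Depth and optional near-camera fades apply only at Medium and High quality; opacity also fades on downward-facing normals.
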Material /Game/Creative/Devices/DeviceZone_Base/Materials/M_CP_DeviceZone_LightBeams
graph {
 properties {
  bEnableResponsiveAA=true
  bUsedWithMeshParticles=true
  blendMode=BLEND_AlphaComposite
  pixelDepthOffsetMode=PDOM_Legacy
  shadingModel=MSM_Unlit
  twoSided=true
  usedShadingModels=Unlit
 }
 nodes [
  n1 = VectorParameter() {bUseCustomPrimitiveData=true, parameterName=CustomPrimitiveColor, primitiveDataIndex=28}
  n2 = MaterialFunctionCall(Float2=TextureCoordinate()) {materialFunction=&BreakOutFloat2Components}
  n3 = Constant() {r=0.5}
  n4 = SphereMask(A=@n2.R, B=@n3, Radius=@n3, Hardness=Constant())
  n5 = ScalarParameter() {defaultValue=0.3, parameterName=CloudNoiseScaleU}
  n6 = ScalarParameter() {defaultValue=0.5, parameterName=CloudNoiseScaleV}
  n7 = Panner(
   Coordinate=Multiply(A=AppendVector(A=@n4, B=@n2.G), B=AppendVector(A=@n5, B=@n6))
   Time=Add(A=ScalarParameter() {bUseCustomPrimitiveData=true, parameterName=Random}, B=Time())
   Speed=AppendVector(A=ScalarParameter() {parameterName=CloudNoisePanSpeedU}, B=ScalarParameter() {defaultValue=-0.0122, parameterName=CloudNoisePanSpeedV})
  ) {bFractionalPart=true}
  n8 = Multiply(
   A=Subtract(A=TextureSample(UVs=@n7) {texture=&AtmosphericCloudNoise01}.R) {constB=0.5}
   B=ScalarParameter() {defaultValue=0.25, parameterName=CloudNoiseScale}
  )
  n9 = TextureCoordinate()
  n10 = StaticSwitchParameter(True=OneMinus(None=@n9), False=@n9) {parameterName=FlipUVs}
  n11 = ScalarParameter() {defaultValue=0.9, parameterName=FalloffRadius}
  n12 = Divide(A=ScalarParameter() {parameterName=Length}, B=ScalarParameter() {defaultValue=400, parameterName=FalloffBaseLength})
  n13 = SphereMask(
   A=ComponentMask(None=@n10) {g=true}
   B=Constant() {r=1}
   Radius=Saturate(None=Divide(A=@n11, B=@n12))
   Hardness=ScalarParameter() {defaultValue=0.025, parameterName=FalloffHardness}
  )
  n14 = Power(Base=@n13, Exp=ScalarParameter() {defaultValue=3.5, parameterName=MaskExponent})
  n15 = MaterialFunctionCall(
   Brightness=ScalarParameter() {defaultValue=1, parameterName=FresnelBrightness}
   Exponent=LinearInterpolate(
    A=ScalarParameter() {defaultValue=4.5, parameterName=FresnelExponentHigh}
    B=ScalarParameter() {defaultValue=1, parameterName=FresnelExponentLow}
    Alpha=@n14
   )
  ) {materialFunction=&MF_Fresnel_FX}
  n16 = ScalarParameter() {defaultValue=0.6, parameterName=OpacityHigh}
  n17 = Multiply(
   A=ScalarParameter() {defaultValue=100, parameterName=Steps}
   B=Divide(A=ScalarParameter() {defaultValue=300, parameterName=Length}, B=ScalarParameter() {defaultValue=300, parameterName=ScanlineBaseLength})
  )
  n18 = ScalarParameter() {defaultValue=500, parameterName=MaxSteps}
  n19 = Frac(None=Multiply(A=ComponentMask(None=@n10) {g=true}, B=Clamp(None=@n17, Max=@n18) {clampMode=CMODE_ClampMax}))
  n20 = Constant() {r=0.5}
  n21 = ScalarParameter() {defaultValue=0.3, parameterName=ScanlinesRadius}
  n22 = ScalarParameter() {parameterName=ScanlinesHardness}
  n23 = Subtract(A=ComponentMask(None=@n10) {g=true}) {constB=0.5}
  n24 = ScalarParameter() {defaultValue=1.5, parameterName=Speed}
  n25 = Multiply(A=Time() {bOverride_Period=true, period=1.5}, B=@n24)
  n26 = Divide(B=@n24)
  n27 = Multiply(A=OneMinus(None=Multiply(A=@n25, B=@n26))) {constB=4}
  n28 = ScalarParameter() {defaultValue=1.5, parameterName=SweepRadius}
  n29 = ScalarParameter() {parameterName=SweepHardness}
  n30 = Multiply(
   A=Multiply(A=SphereMask(A=@n19, B=@n20, Radius=@n21, Hardness=@n22), B=SphereMask(A=@n23, B=@n27, Radius=@n28, Hardness=@n29))
   B=ScalarParameter() {defaultValue=0.55, parameterName=ScanlinesOpacity}
  )
  n31 = Constant()
  n32 = Multiply(
   A=@n15.Fresnel Double Sided
   B=Add(A=LinearInterpolate(B=@n16, Alpha=@n14), B=StaticSwitchParameter(True=@n30, False=@n31) {defaultValue=true, parameterName=UseScanlines})
  )
  n33 = OneMinus(None=@n14)
  n34 = VectorParameter() {parameterName=NoiseUVTileA, defaultValue=<rgba 2 0.09 0 0>}
  n35 = Panner(
   Coordinate=Multiply(A=@n10, B=ComponentMask(None=@n34.RGB) {g=true, r=true})
   Speed=ComponentMask(None=VectorParameter() {parameterName=NoisePanSpeedA, defaultValue=<rgba 0.36 0.11 0 0>}.RGB) {g=true, r=true}
  ) {bFractionalPart=true}
  n36 = VectorParameter() {parameterName=NoiseUVTileB, defaultValue=<rgba 3 0.11 0 0>}
  n37 = Panner(
   Coordinate=Multiply(A=@n10, B=ComponentMask(None=@n36.RGB) {g=true, r=true})
   Speed=ComponentMask(None=VectorParameter() {parameterName=NoisePanSpeedB, defaultValue=<rgba -0.63 0 0.1335 0>}.RGB) {g=true, r=true}
  ) {bFractionalPart=true, speedX=-0.026}
  n38 = Add(
   A=Subtract(A=TextureSample(UVs=@n35) {samplerType=SAMPLERTYPE_LinearColor, texture=&T_NoisePack_LinearColor}.R) {constB=0.5}
   B=Subtract(A=TextureSample(UVs=@n37) {samplerType=SAMPLERTYPE_LinearColor, texture=&T_NoisePack_LinearColor}.G) {constB=0.5}
  )
  n39 = Multiply(A=@n15.Fresnel Double Sided, B=Multiply(A=Multiply(A=@n33, B=@n38), B=ScalarParameter() {defaultValue=1.5, parameterName=NoiseScale}))
  n40 = Multiply(A=Add(A=@n32, B=@n39), B=@n13)
  n41 = StaticSwitchParameter(True=Add(A=@n8, B=@n40), False=@n40) {parameterName=UseCloudNoise}
  n42 = LinearInterpolate(
   A=VectorParameter() {parameterName=ColorLow, defaultValue=<rgba 0.0833 0.0247 0.8 1>}.RGB
   B=VectorParameter() {parameterName=ColorHigh, defaultValue=<rgba 0.1438 0.6079 2 1>}.RGB
   Alpha=Saturate(None=@n41)
  )
  n43 = ScalarParameter() {defaultValue=30, parameterName=FadeDistance}
  n44 = DepthFade(Opacity=@n41, FadeDistance=@n43)
  n45 = QualitySwitch(Default=@n41, Low=@n41, High=@n44, Medium=@n44)
  n46 = ScalarParameter() {defaultValue=1, parameterName=Fade}
  n47 = VertexColor()
  n48 = VectorParameter() {parameterName=VertexColorMask, defaultValue=<rgba 0 0 0 1>}
  n49 = StaticSwitchParameter(True=DotProduct(A=AppendVector(A=@n47, B=@n47), B=AppendVector(A=@n48.RGB, B=@n48.RGB)), False=Constant() {r=1}) {parameterName=MaskVertexColor}
  n50 = Multiply(
   A=Multiply(A=StaticSwitchParameter(True=@n1.RGB, False=@n42) {parameterName=ColorUsesCustomPrimitiveVec4}, B=Multiply(A=@n45, B=@n46))
   B=@n49
  )
  n51 = ParticleColor()
  n52 = Multiply(A=@n13, B=Add(A=@n39, B=@n32))
  n53 = StaticSwitchParameter(True=Add(A=@n52, B=@n8), False=@n52) {parameterName=UseCloudNoise}
  n54 = DepthFade(Opacity=@n53, FadeDistance=@n43)
  n55 = Multiply(A=Multiply(A=@n46, B=QualitySwitch(Default=@n53, Low=@n53, High=@n54, Medium=@n54)), B=@n49)
  n56 = Multiply(A=@n51.A, B=@n55)
  n57 = StaticSwitchParameter(True=Multiply(A=Multiply(A=@n50, B=@n51.RGB), B=@n56), False=@n50) {parameterName=UseParticleColor}
  n58 = Constant() {r=1}
  n59 = MaterialFunctionCall(Distance=ScalarParameter() {defaultValue=100, parameterName=NearCameraFadeDistance}) {materialFunction=&Near_Camera_Fade}
  n60 = QualitySwitch(Default=@n58, Low=@n58, High=@n59.Result, Medium=@n59.Result)
  MaterialFunctionCall(Emissive=StaticSwitchParameter(True=Multiply(A=@n57, B=@n60), False=@n57) {parameterName=UseNearCameraFade}) {materialFunction=&MF_FX_EmissiveAdjustments}
  n61 = StaticSwitchParameter(True=@n56, False=@n55) {parameterName=UseParticleColor}
  n62 = Multiply(A=ComponentMask(None=VertexNormalWS()) {b=true}) {constB=-1}
  Multiply(A=StaticSwitchParameter(True=Multiply(A=@n61, B=@n60), False=@n61) {parameterName=UseNearCameraFade}, B=OneMinus(None=Saturate(None=@n62)))
 ]
}
