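# Translucent, unlit, two-sided material for the placed bounds of the Creative fishing zone device, styled as caustic-lit water.
# Two WaterFlowMap phases sample a 6x6 caustics flipbook with explicit derivatives, scaled by object size, and crossfade.
# Fresnel and a height gradient shape background color and opacity; camera distance and facing modulate caustics.
# Fluid-simulation and depth-fade caustic branches end without feeding that color and opacity blend.
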Material /Game/Creative/Devices/DeviceZone_Base/Materials/M_CP_FishingZone_Bounds_Placed
graph {
 properties {blendMode=BLEND_Translucent, pixelDepthOffsetMode=PDOM_Legacy, shadingModel=MSM_Unlit, twoSided=true, usedShadingModels=Unlit}
 nodes [
  n1 = Constant() {r=50}
  n2 = ScalarParameter() {defaultValue=0.5, group=Caustics, parameterName=Caustics Phase Fade Time}
  n3 = Divide(A=Divide(A=Constant2Vector(), B=@n1), B=@n2)
  n4 = MaterialFunctionCall() {materialFunction=&ObjectScale}
  n5 = Multiply(A=Constant() {r=1}, B=@n4.Scale X)
  n6 = Multiply(A=@n4.Scale Y, B=Divide(A=@n4.Scale Z, B=@n4.Scale Z))
  n7 = Multiply(A=Multiply(A=TextureCoordinate(), B=AppendVector(A=@n5, B=@n6)), B=Constant() {r=1})
  n8 = MaterialFunctionCall!(
   Diffuse
   Normal Map
   "Flow Vector Map (see tooltip)"=@n3
   UV texture 2 offset and Scale=VectorParameter() {parameterName=Flowmap UV2 Offset and Scale, defaultValue=<rgba 0 0 1 0>}.RGB
   "Time (see tooltip)"=MaterialFunctionCall(Speed=@n2) {materialFunction=&TimeWithSpeedVariable}.Time
   UVs=@n7
  ) {materialFunction=/FortWater/Materials/Functions/WaterFlowMap}
  n9 = Frac(None=@n8.UV A)
  n10 = Constant() {r=6}
  n11 = Constant() {r=6}
  n12 = Multiply(A=Time(), B=ScalarParameter() {defaultValue=0.75, parameterName=CausticAnimationPlayRate})
  n13 = Add(A=@n12) {constB=0.5}
  n14 = MaterialFunctionCall!(1D Index=Multiply(A=Multiply(A=@n11, B=@n11), B=Frac(None=@n13)), Cells XY=@n11) {materialFunction=/Engine/Functions/Engine_MaterialFunctions01/Coordinates/1Dto2DIndex}
  n15 = TextureObject() {samplerType=SAMPLERTYPE_LinearColor, texture=/FortWater/Textures/T_Siren_Caustics_03}
  n16 = Divide(A=@n7) {constB=6}
  n17 = DDX(Value=@n16)
  n18 = DDY(Value=@n16)
  n19 = TextureSample!(UVs=Add(A=Divide(A=@n9, B=@n10), B=Divide(A=@n14.2D Index, B=@n11)), Tex=@n15, "DDX(UVs)"=@n17, "DDY(UVs)"=@n18, Apply View MipBias) {mipValueMode=TMVM_Derivative}
  n20 = Frac(None=@n8.UV B)
  n21 = Constant() {r=6}
  n22 = MaterialFunctionCall(1D Index=Multiply(A=Multiply(A=@n21, B=@n21), B=Frac(None=@n12)), Cells XY=@n21) {materialFunction=&1Dto2DIndex}
  n23 = TextureSample(UVs=Add(A=Divide(A=@n20, B=@n10), B=Divide(A=@n22.2D Index, B=@n21)), Tex=@n15, "DDX(UVs)"=@n17, "DDY(UVs)"=@n18) {mipValueMode=TMVM_Derivative}
  n24 = ComponentMask(None=@n8.Distortion) {b=true}
  n25 = LinearInterpolate(A=@n19.RGB, B=@n23.RGB, Alpha=@n24)
  n26 = ComponentMask(None=WorldPosition().XYZ) {g=true, r=true}
  n27 = VectorParameter() {parameterName=V, defaultValue=<rgba 0 -0.1 0 0>}
  n28 = Add(A=@n26, B=Multiply(B=ComponentMask(None=@n27.RGB) {g=true, r=true}))
  n29 = ComponentMask(None=CollectionParameter() {collection=/Water/Materials/MPC/MPC_Water, parameterName=SimLocation}) {g=true, r=true}
  n30 = Add(A=Divide(A=Subtract(A=@n28, B=@n29), B=CollectionParameter() {collection=&MPC_Water, parameterName=FluidSimSize})) {constB=0.5}
  n31 = TextureSampleParameter2D(UVs=Saturate(None=@n30)) {group=Fluid Simulation, parameterName=NormalAndHeight, samplerSource=SSM_Wrap_WorldGroupSettings, texture=&Black}
  n32 = DeriveNormalZ(InXY=ComponentMask(None=@n31.RGB) {g=true, r=true})
  n33 = Add(A=@n30, B=ComponentMask(None=VectorParameter() {parameterName=UV Offset, defaultValue=<rgba 0.01 0.01 0 0>}.RGB) {g=true, r=true})
  n34 = TextureSampleParameter2D(UVs=Saturate(None=@n33)) {group=Fluid Simulation, parameterName=NormalAndHeight, samplerSource=SSM_Wrap_WorldGroupSettings, texture=&Black}
  n35 = DeriveNormalZ(InXY=ComponentMask(None=@n34.RGB) {g=true, r=true})
  n36 = Subtract(A=ComponentMask(None=OneMinus(None=@n32)) {b=true}, B=ComponentMask(None=OneMinus(None=@n35)) {b=true})
  Add(A=@n25, B=Multiply(A=Abs(None=@n36), B=ScalarParameter() {defaultValue=1, parameterName=Simulation Caustics}))
  Clamp(None=@n33)
  ComponentMask(None=WorldPosition().XYZ) {g=true, r=true}
  Constant4Vector() {constant=<rgba 0 0 0 1>}
  Divide(A=@n26, B=@n1)
  n37 = Subtract(A=WorldPosition().XYZ, B=ActorPositionWS())
  n38 = ScalarParameter() {defaultValue=100, parameterName=HeightGradientDivisor}
  n39 = OneMinus(None=Saturate(None=Divide(A=@n37, B=@n38)))
  n40 = Add(
   A=MaterialFunctionCall(
    Brightness=ScalarParameter() {defaultValue=0.65, parameterName=FresnelBrightness}
    Exponent=ScalarParameter() {defaultValue=4, parameterName=FresnelExponent}
   ) {materialFunction=&MF_Fresnel_FX}.Fresnel Inverted Double Sided
   B=Multiply(
    A=MaterialFunctionCall(Float3=@n39) {materialFunction=&BreakOutFloat3Components}.B
    B=ScalarParameter() {defaultValue=0.5, parameterName=HeightGradientScale}
   )
  )
  n41 = VectorParameter() {group=BG, parameterName=BG Bottom Glow Color, sortPriority=2, defaultValue=<rgba 0.2212 0.3969 0.4917 1>}
  n42 = ScalarParameter() {defaultValue=1, parameterName=Caustics Intensity}
  n43 = Multiply(A=Desaturation(None=Multiply(A=@n25, B=@n42)), B=Constant() {r=1.5})
  n44 = Constant3Vector() {desc=Caustic Color, constant=<rgba 1 0.9184 0.5 1>}
  n45 = CameraPositionWS()
  n46 = WorldPosition()
  n47 = Divide(
   A=Add(A=Distance(A=@n45, B=@n46.XYZ), B=ScalarParameter() {defaultValue=-100, parameterName=DistanceBias})
   B=ScalarParameter() {defaultValue=4000, parameterName=FadeScale}
  )
  n48 = ScalarParameter() {defaultValue=0.25, group=BLOCKS, parameterName=BlocksBackfaceOpacity, sortPriority=11}
  n49 = TwoSidedSign()
  n50 = Multiply(A=OneMinus(None=Saturate(None=@n47)), B=LinearInterpolate(A=Multiply(A=@n48), Alpha=Saturate(None=@n49)))
  LinearInterpolate(
   A=LinearInterpolate(
    A=VectorParameter() {group=BG, parameterName=BG Color, sortPriority=1, defaultValue=<rgba 0.0382 0.2086 0.3005 1>}.RGB
    B=@n41.RGB
    Alpha=@n40
   )
   B=@n43
   Alpha=Multiply(A=Multiply(A=@n43, B=@n44), B=@n50)
  )
  LinearInterpolate(
   A=LinearInterpolate(A=Constant() {desc=BG Color Opacity, r=0.25}, B=Constant() {desc=Caustic Opacityl, r=1}, Alpha=Multiply(A=@n43, B=@n50))
   Alpha=@n40
  )
  MaterialFunctionCall(Speed=ScalarParameter() {defaultValue=0.5, group=Caustics, parameterName=WaterCausticsTimeSpeed}) {materialFunction=&TimeWithSpeedVariable}
  n51 = Multiply(A=@n25, B=ScalarParameter() {defaultValue=1, group=Caustics, parameterName=Caustics Brightness})
  n52 = ScalarParameter() {defaultValue=0.2, group=Caustics, parameterName=Caustics Add}
  n53 = Divide(A=Constant() {r=300}, B=ScalarParameter() {defaultValue=256, group=Caustics, parameterName=Caustics Gradient Length})
  n54 = ScalarParameter() {defaultValue=7500, group=Caustics, parameterName=Caustics Fade Distance}
  n55 = LinearInterpolate(
   A=LinearInterpolate(
    A=ScalarParameter() {defaultValue=1, group=Caustics, parameterName=Caustics Fadeout Brightness}
    B=Add(A=@n51, B=@n52)
    Alpha=Saturate(None=@n53)
   ) {constA=0.75}
   B=ScalarParameter() {defaultValue=1.5, parameterName=Caustics Distance brightness}
   Alpha=MaterialFunctionCall(Fade Length=@n54, Fade Offset=@n54) {materialFunction=&CameraDepthFade}.Result
  )
  n56 = Constant() {r=1}
  n57 = QualitySwitch(Default=FeatureLevelSwitch(Default=@n55, ES2=@n56, ES3_1=@n56, SM4_REMOVED=@n55), Low=@n56)
  MaterialFunctionCall(In=Saturate(None=Desaturation(None=@n57)), Contrast=Constant() {r=0.5}) {materialFunction=&CheapContrast}
  Multiply(A=@n3)
  Multiply(A=@n16, B=ScalarParameter() {defaultValue=5, parameterName=MipScale})
  Power(Base=LinearInterpolate(A=@n23.R, B=@n19.R, Alpha=@n24), Exp=ScalarParameter() {defaultValue=2, parameterName=Exp})
  ScalarParameter()
 ]
}
