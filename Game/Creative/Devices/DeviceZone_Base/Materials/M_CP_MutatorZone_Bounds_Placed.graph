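# Translucent, unlit, two-sided material for mutator zone device bounds in Creative, showing a tinted volume with animated blocky flickers.
# Four noise layers, projected in world space by face normal, jump on stepped time and pulse with LinearSine to form colored blocks that fade with camera distance and on backfaces.
# Fresnel and an object-scale-compensated bottom gradient drive glow and opacity.
# Static switches handle sphere versus flat-faced shapes and optional outline fresnel.
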Material /Game/Creative/Devices/DeviceZone_Base/Materials/M_CP_MutatorZone_Bounds_Placed
graph {
 properties {blendMode=BLEND_Translucent, pixelDepthOffsetMode=PDOM_Legacy, shadingModel=MSM_Unlit, twoSided=true, usedShadingModels=Unlit}
 nodes [
  n1 = Constant() {r=1.5}
  n2 = MaterialFunctionCall(Float3=Divide(A=WorldPosition().XYZ, B=ScalarParameter() {defaultValue=450, parameterName=WorldScale})) {materialFunction=&BreakOutFloat3Components}
  n3 = VertexNormalWS()
  n4 = MaterialFunctionCall(Float3=Abs(None=Round(None=@n3))) {materialFunction=&BreakOutFloat3Components}
  n5 = LinearInterpolate(
   A=LinearInterpolate(A=AppendVector(A=@n2.G, B=@n2.G), B=AppendVector(A=@n2.R, B=@n2.R), Alpha=@n4.G)
   B=AppendVector(A=@n2.R, B=@n2.R)
   Alpha=@n4.B
  )
  n6 = Constant()
  n7 = Constant() {r=0.25}
  n8 = Multiply(A=Time(), B=Constant() {r=0.53})
  n9 = Panner(Coordinate=Add(A=Multiply(A=@n5, B=@n1) {constB=0.25}, B=AppendVector(A=@n6, B=@n7)), Time=Ceil(None=@n8)) {bFractionalPart=true, speedX=-0.25, speedY=-0.25}
  n10 = Multiply(A=Desaturation(None=TextureSample(UVs=@n9) {samplerType=SAMPLERTYPE_LinearColor, texture=/Game/Packages/Fortress_Sky/Materials/Noise32x322_nearestNeighboor}.RGB)) {constB=10}
  n11 = Saturate(None=Ceil(None=OneMinus(None=@n10)))
  n12 = StaticBool()
  n13 = Multiply(
   A=@n11
   B=MaterialFunctionCall(Value=@n8, Sine Phase=@n12) {materialFunction=&LinearSine}."Rounded Linear Sine "
  )
  n14 = Constant() {r=1.5}
  n15 = Constant() {r=0.5}
  n16 = Constant()
  n17 = Multiply(A=Time(), B=Constant() {r=0.96})
  n18 = Panner(Coordinate=Add(A=Multiply(A=@n5, B=@n14) {constB=0.25}, B=AppendVector(A=@n15, B=@n16)), Time=Ceil(None=@n17)) {bFractionalPart=true, speedX=-0.25, speedY=-0.25}
  n19 = Multiply(A=Desaturation(None=TextureSample(UVs=@n18) {samplerType=SAMPLERTYPE_LinearColor, texture=&Noise32x322_nearestNeighboor}.RGB)) {constB=10}
  n20 = Saturate(None=Ceil(None=OneMinus(None=@n19)))
  n21 = Multiply(
   A=@n20
   B=MaterialFunctionCall(Value=@n17, Sine Phase=@n12) {materialFunction=&LinearSine}."Rounded Linear Sine "
  )
  n22 = Constant() {r=1.5}
  n23 = Constant() {r=0.25}
  n24 = Constant()
  n25 = Multiply(A=Time(), B=Constant() {r=1.43})
  n26 = Panner(Coordinate=Add(A=Multiply(A=@n5, B=@n22) {constB=0.25}, B=AppendVector(A=@n23, B=@n24)), Time=Ceil(None=@n25)) {bFractionalPart=true, speedX=-0.25, speedY=-0.25}
  n27 = Multiply(A=Desaturation(None=TextureSample(UVs=@n26) {samplerType=SAMPLERTYPE_LinearColor, texture=&Noise32x322_nearestNeighboor}.RGB)) {constB=10}
  n28 = Saturate(None=Ceil(None=OneMinus(None=@n27)))
  n29 = MaterialFunctionCall(Value=@n25, Sine Phase=@n12) {materialFunction=&LinearSine}
  n30 = Constant() {r=1.5}
  n31 = Constant()
  n32 = Constant() {r=0.5}
  n33 = Multiply(A=Time(), B=Constant() {r=1.94})
  n34 = Panner(Coordinate=Add(A=Multiply(A=@n5, B=@n30) {constB=0.25}, B=AppendVector(A=@n31, B=@n32)), Time=Ceil(None=@n33)) {bFractionalPart=true, speedX=-0.25, speedY=-0.25}
  n35 = Multiply(A=Desaturation(None=TextureSample(UVs=@n34) {samplerType=SAMPLERTYPE_LinearColor, texture=&Noise32x322_nearestNeighboor}.RGB)) {constB=10}
  n36 = Saturate(None=Ceil(None=OneMinus(None=@n35)))
  n37 = Add(
   A=Add(A=Add(A=@n13, B=@n21), B=Multiply(A=@n28, B=@n29."Rounded Linear Sine "))
   B=Multiply(
    A=@n36
    B=MaterialFunctionCall(Value=@n33, Sine Phase=@n12) {materialFunction=&LinearSine}."Rounded Linear Sine "
   )
  )
  n38 = CameraPositionWS()
  n39 = WorldPosition()
  n40 = Divide(
   A=Add(A=Distance(A=@n38, B=@n39.XYZ), B=ScalarParameter() {defaultValue=-100, parameterName=DistanceBias})
   B=ScalarParameter() {defaultValue=4000, parameterName=FadeScale}
  )
  n41 = ScalarParameter() {defaultValue=0.1, group=BLOCKS, parameterName=BlocksBackfaceOpacity, sortPriority=11}
  n42 = TwoSidedSign()
  n43 = Multiply(A=OneMinus(None=Saturate(None=@n40)), B=LinearInterpolate(A=Multiply(A=@n41), Alpha=Saturate(None=@n42)))
  n44 = OneMinus(None=ComponentMask(None=VertexNormalWS()) {b=true})
  n45 = MaterialFunctionCall(
   Brightness=ScalarParameter() {defaultValue=0.65, group=BG, parameterName=FresnelBrightness}
   Exponent=ScalarParameter() {defaultValue=4, group=BG, parameterName=FresnelExponent}
  ) {materialFunction=&MF_Fresnel_FX}
  n46 = Constant() {r=1}
  n47 = ComponentMask(None=TextureCoordinate() {coordinateIndex=2}) {r=true}
  n48 = TextureCoordinate() {coordinateIndex=2}
  n49 = AppendVector(
   A=Divide(
    A=Constant() {r=1}
    B=MaterialFunctionCall() {materialFunction=&ObjectScale}.Scale Y
   )
   B=Divide(
    A=Constant() {desc="0.85 is default, not 1", r=2}
    B=MaterialFunctionCall() {materialFunction=&ObjectScale}.Scale Z
   )
  )
  n50 = Constant() {r=1}
  n51 = Subtract(A=Add(A=Divide(A=@n48, B=@n49), B=@n50), B=Divide(A=@n50, B=@n49) {desc=Scale the center offset by the scaling factor}) {desc=Re offset to maintain center}
  n52 = Constant()
  n53 = Constant() {r=1}
  n54 = ComponentMask(None=TextureCoordinate() {coordinateIndex=1}) {r=true}
  n55 = TextureCoordinate() {coordinateIndex=1}
  n56 = AppendVector(
   A=Divide(
    A=Constant() {r=1}
    B=MaterialFunctionCall() {materialFunction=&ObjectScale}.Scale X
   )
   B=Divide(
    A=Constant() {desc="0.85 is default, not 1", r=2}
    B=MaterialFunctionCall() {materialFunction=&ObjectScale}.Scale Z
   )
  )
  n57 = Constant() {r=1}
  n58 = Subtract(A=Add(A=Divide(A=@n55, B=@n56), B=@n57), B=Divide(A=@n57, B=@n56) {desc=Scale the center offset by the scaling factor}) {desc=Re offset to maintain center}
  n59 = Constant()
  n60 = TextureSample(UVs=Add(A=If(A=@n46, B=@n47, "A > B"=@n51, "A < B"=@n52) {desc=0}, B=If(A=@n53, B=@n54, "A > B"=@n58, "A < B"=@n59) {desc=0})) {
   samplerSource=SSM_Clamp_WorldGroupSettings
   texture=/Game/Creative/Devices/DeviceZone_Base/Materials/Textures/T_CP_DeviceZone_Bounds_01_BarrierPlacer_D
  }
  n61 = Add(
   A=StaticSwitchParameter(True=@n45.Fresnel Inverted Double Sided, False=Multiply(A=@n44, B=@n45.Fresnel Inverted Double Sided)) {parameterName=IsSphere}
   B=Multiply(A=@n60.G, B=ScalarParameter() {defaultValue=0.4, group=BG, parameterName=BottomGradientOpacity}) {constB=0.5}
  )
  n62 = LinearInterpolate(
   A=LinearInterpolate(
    A=ScalarParameter() {defaultValue=0.1, group=BG, parameterName=BG Color Opacity, sortPriority=3}
    B=ScalarParameter() {defaultValue=0.1, group=BLOCKS, parameterName=Blocks Opacity, sortPriority=10}
    Alpha=Multiply(A=@n37, B=@n43)
   )
   Alpha=@n61
  )
  n63 = ComponentMask(None=VertexNormalWS()) {b=true}
  n64 = OneMinus(None=Saturate(None=Multiply(A=@n63) {constB=-1}))
  Reroute(None=StaticSwitchParameter(True=@n62, False=Multiply(A=@n62, B=@n64)) {parameterName=IsSphere})
  n65 = ScalarParameter() {defaultValue=2, group=Outline Fresnel, parameterName=Outline Fresnel Exponent, sortPriority=2}
  n66 = VectorParameter() {group=Outline Fresnel, parameterName=Outline Fresnel Color, sortPriority=1, defaultValue=<rgba 1 1 1 1>}
  n67 = Multiply(
   A=MaterialFunctionCall(In=Fresnel(ExponentIn=@n65), Contrast=ScalarParameter() {group=Outline Fresnel, parameterName=Outline Fresnel Contrast, sortPriority=3}) {materialFunction=&CheapContrast}.Result
   B=@n66.RGB
  )
  n68 = ScalarParameter() {defaultValue=1, group=Outline Fresnel, parameterName=Outline Fresnel Intensity, sortPriority=4}
  n69 = VectorParameter() {group=BG, parameterName=BG Bottom Glow and Fresnel Color, sortPriority=2, defaultValue=<rgba 0 1 0 1>}
  n70 = Multiply(
   A=VectorParameter() {group=BLOCKS, parameterName=Color 3, sortPriority=3, defaultValue=<rgba 0.0212 0.5711 0.4397 1>}.RGB
   B=ScalarParameter() {defaultValue=1, group=BLOCKS, parameterName=Color 3 Intensity Multiplier, sortPriority=7}
  )
  n71 = Multiply(
   A=VectorParameter() {group=BLOCKS, parameterName=Color 2, sortPriority=2, defaultValue=<rgba 0.2307 0.5841 0.4508 1>}.RGB
   B=ScalarParameter() {defaultValue=1, group=BLOCKS, parameterName=Color 2 Intensity Multiplier, sortPriority=6}
  )
  n72 = VectorParameter() {group=BLOCKS, parameterName=Color 1, sortPriority=1, defaultValue=<rgba 0.0561 0.4793 0.3005 1>}
  n73 = ScalarParameter() {defaultValue=1, group=BLOCKS, parameterName=Color 1 Intensity Multipler, sortPriority=5}
  n74 = LinearInterpolate(
   A=LinearInterpolate(A=Multiply(A=@n70, B=@n11), B=Multiply(A=@n71, B=@n20), Alpha=@n20)
   B=Multiply(A=Multiply(A=@n72.RGB, B=@n73), B=@n28)
   Alpha=@n28
  )
  n75 = Multiply(
   A=Multiply(
    A=VectorParameter() {group=BLOCKS, parameterName=Color 4, sortPriority=4, defaultValue=<rgba 0.0299 0.3583 0.2798 1>}.RGB
    B=ScalarParameter() {defaultValue=1, group=BLOCKS, parameterName=Color 4 Intensity Multiplier, sortPriority=8}
   )
   B=@n36
  )
  n76 = Add(A=Add(A=@n11, B=@n20), B=@n28)
  n77 = LinearInterpolate(
   A=LinearInterpolate(A=VectorParameter() {group=BG, parameterName=BG Color, sortPriority=1, defaultValue=<rgba 0 0.2375 0.0931 1>}.RGB, B=@n69.RGB, Alpha=@n61)
   B=Multiply(
    A=LinearInterpolate(A=@n74, B=@n75, Alpha=@n36)
    B=ScalarParameter() {defaultValue=2, group=BLOCKS, parameterName=Overall Blocks Intensity, sortPriority=9}
   ) {constB=0.85}
   Alpha=Multiply(A=Add(A=@n76, B=@n75), B=@n43)
  )
  StaticSwitchParameter(True=Add(A=Multiply(A=@n67, B=@n68), B=@n77), False=@n77) {parameterName=UseOutlineFresnel}
 ]
}
